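; Gameplay cue for the Cosmos rifle's melee impact, playing camera shake and particle effects when its burst fires.
; Plays a world camera shake at the target pawn's mesh and snaps the cue's particle component to that mesh's spine_03.
; Spawns an impact Niagara system at the muzzle of player 0's current weapon, not one resolved from cue parameters.
; Sets the attached effect's BeamEnd_0 to that muzzle location.
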(fn OnBurst(EitherVFX-preferthisversion)
  (params (MyTarget :type Actor%20Object%20Reference) (Parameters :type Gameplay%20Cue%20Parameters%20Structure%20%28by%20ref%29) (ParticleComponents :type Array%20of%20FXSystem%20Component%20Object%20References) (AudioComponents :type Array%20of%20Audio%20Component%20Object%20References) (BurstCameraShakeInstance :type Camera%20Shake%20Base%20Object%20Reference) (BurstDecalInstance :type Decal%20Actor%20Object%20Reference))
  (bind _output (Utilities|Array|Get(acopy) ParticleComponents))
  (bind _asfort_player_pawn (Utilities|Casting|CastToFortPlayerPawn MyTarget))
  (Camera|PlayWorldCameraShake 0 "/CosmosGameplay/Items/CosmosRifle/FX/BP_Camera_Shake_Cosmos_rifle_Melee_impact.BP_Camera_Shake_Cosmos_rifle_Melee_impact_C" (Transformation|GetWorldLocation (Variables|/Script/Engine.Character|GetMesh _asfort_player_pawn)) 3e+02 5e+02)
  (bind _asfort_player_pawn_2 (Utilities|Casting|CastToFortPlayerPawn MyTarget))
  (Transformation|AttachComponentToComponent _output (Variables|/Script/Engine.Character|GetMesh _asfort_player_pawn_2) "spine_03" "SnapToTarget" "SnapToTarget" "SnapToTarget" true)
  (bind _asfort_player_pawn_3 (Utilities|Casting|CastToFortPlayerPawn (Game|GetPlayerPawn 0)))
  (Niagara|SpawnSystematLocation (Variables|/Script/FortniteGame.FortWeapon|GetWeaponMesh (Variables|/Script/FortniteGame.FortPawn|GetCurrentWeapon _asfort_player_pawn_3)) "/CosmosGameplay/Items/CosmosRifle/FX/NS_cosmos_rifle_impact.NS_cosmos_rifle_impact" (Transformation|GetSocketLocation (Variables|/Script/FortniteGame.FortWeapon|GetWeaponMesh (Variables|/Script/FortniteGame.FortPawn|GetCurrentWeapon _asfort_player_pawn_3)) "Muzzle"))
  (bind _asniagara_particle_system_component (Utilities|Casting|CastToNiagaraComponent _output))
  (Niagara|SetNiagaraVariable(Vector3) _asniagara_particle_system_component "BeamEnd_0" (Transformation|GetSocketLocation (Variables|/Script/FortniteGame.FortWeapon|GetWeaponMesh (Variables|/Script/FortniteGame.FortPawn|GetCurrentWeapon _asfort_player_pawn_3)) "Muzzle"))
  (return))
